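# Default Lit material for the MaxisSE wheel skeletal mesh, with animated, color-customizable glow and scalar base color, roughness and a rim normal map.
# Offset copies of one panning texture, scaled by TireEmissiveGlow, are blended through mask-texture channels.
# Vertex-color blue blends in a rotated, dimmed copy of the scroll.
# CustomColor tints the result, and TreadLight blends it toward white by the blue mask times that color.
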Material /VehicleCosmetics/Wheels/MaxisSE/Materials/M_Wheel_MaxisSE
graph {
 properties {
  bAutomaticallySetUsageInEditor=false
  bUsedWithSkeletalMesh=true
  previewMesh=/VehicleCosmetics/Wheels/MaxisSE/Meshes/SK_Wheel_MaxisSE
  usedShadingModels=Default Lit
 }
 nodes [
  n1 = ScalarParameter() {defaultValue=1, parameterName=TireEmissiveGlow}
  n2 = TextureSample(UVs=Panner(Coordinate=TextureCoordinate() {uTiling=0.33}) {speedX=0.5}) {
   samplerSource=SSM_Wrap_WorldGroupSettings
   samplerType=SAMPLERTYPE_LinearColor
   texture=/VehicleCosmetics/Wheels/MaxisSE/Textures/T_MaxisSE_Rim_RGB
  }
  n3 = Multiply(A=@n1, B=@n2.R)
  n4 = Constant3Vector() {constant=<rgba 1 1 1 0>}
  n5 = Multiply(A=@n4, B=@n1)
  n6 = TextureCoordinate() {uTiling=0.33}
  n7 = AppendVector(A=Constant() {r=-0.3}, B=Constant())
  n8 = TextureSample(UVs=Panner(Coordinate=Add(A=@n6, B=@n7)) {speedX=0.4}) {
   samplerSource=SSM_Wrap_WorldGroupSettings
   samplerType=SAMPLERTYPE_LinearColor
   texture=&T_MaxisSE_Rim_RGB
  }
  n9 = TextureCoordinate() {uTiling=0.33}
  n10 = AppendVector(A=Constant() {r=-0.5}, B=Constant())
  n11 = TextureSample(UVs=Panner(Coordinate=Add(A=@n9, B=@n10)) {speedX=0.4}) {
   samplerSource=SSM_Wrap_WorldGroupSettings
   samplerType=SAMPLERTYPE_LinearColor
   texture=&T_MaxisSE_Rim_RGB
  }
  n12 = TextureSampleParameter2D() {parameterName=RGB, samplerType=SAMPLERTYPE_LinearColor, texture=/VehicleCosmetics/Wheels/MaxisSE/Textures/T_MaxisSE_Rim_M}
  n13 = LinearInterpolate(A=Add(A=Floor(None=@n3), B=Multiply(A=@n5, B=@n8.R)), B=Multiply(A=Multiply(A=@n4, B=@n1), B=@n11.R), Alpha=@n12.A)
  n14 = TextureCoordinate() {uTiling=0.33}
  n15 = AppendVector(A=Constant() {r=-0.2}, B=Constant())
  n16 = TextureSample(UVs=Panner(Coordinate=Add(A=@n14, B=@n15)) {speedX=0.4}) {
   samplerSource=SSM_Wrap_WorldGroupSettings
   samplerType=SAMPLERTYPE_LinearColor
   texture=&T_MaxisSE_Rim_RGB
  }
  n17 = Add(A=@n12.R, B=@n12.R)
  n18 = LinearInterpolate(
   A=LinearInterpolate(A=Clamp(None=@n13) {maxDefault=3}, B=Multiply(A=@n1, B=@n16.R), Alpha=OneMinus(None=@n12.G))
   B=Constant()
   Alpha=OneMinus(None=Add(A=@n12.B, B=@n17))
  )
  n19 = TextureCoordinate() {uTiling=0.2}
  n20 = Constant() {r=0.25}
  n21 = TextureSample(UVs=Panner(Coordinate=MaterialFunctionCall(UVs=@n19, "Rotation Angle (0-1)"=@n20) {materialFunction=&CustomRotator}.Rotated Values) {speedX=0.4}) {
   samplerSource=SSM_Wrap_WorldGroupSettings
   samplerType=SAMPLERTYPE_LinearColor
   texture=&T_MaxisSE_Rim_RGB
  }
  n22 = LinearInterpolate(A=@n21.R, B=Constant()) {constAlpha=0.8}
  n23 = VertexColor()
  n24 = VectorParameter() {parameterName=CustomColor, defaultValue=<rgba 1 0.1135 0 3>}
  n25 = Constant() {r=3}
  LinearInterpolate(
   A=Multiply(A=LinearInterpolate(A=@n18, B=@n22, Alpha=@n23.B), B=Multiply(A=@n24.RGB, B=@n25))
   B=Constant() {r=1}
   Alpha=Multiply(A=Multiply(A=@n12.B, B=@n24.RGB), B=ScalarParameter() {defaultValue=1, parameterName=TreadLight})
  )
  ScalarParameter() {defaultValue=0.5, parameterName=Roughness}
  ScalarParameter() {defaultValue=0.05, parameterName=BaseColor}
  TextureSampleParameter2D() {parameterName=RimNormal, samplerType=SAMPLERTYPE_Normal, texture=/VehicleCosmetics/Wheels/MaxisSE/Textures/T_MaxisSE_Rim_N}
 ]
}
